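# Translucent Niagara ribbon and sprite material, named as the red-and-white trail effect for the PrivateJet pickaxe.
# A sine of UV V creates repeating bands that blend WhiteColor with particle color and feed MF_FX_EmissiveAdjustments.
# An opacity-style mask multiplies particle alpha by an optional U start/end fade, a V edge fade and an inverted radial gradient.
# The RedColor parameter is unconnected, so any red comes from particle color.
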Material /BRCosmetics/Weapons/FORT_Melee/Pickaxe_PrivateJet/FX/M_PrivateJet_RedWhiteTrail
graph {
 properties {
  bUsedWithNiagaraRibbons=true
  bUsedWithNiagaraSprites=true
  blendMode=BLEND_Translucent
  pixelDepthOffsetMode=PDOM_Legacy
  usedShadingModels=Default Lit
 }
 nodes [
  n1 = ParticleColor()
  n2 = Sine(None=ComponentMask(None=TextureCoordinate()) {g=true}) {period=0.25}
  n3 = LinearInterpolate(
   A=VectorParameter() {parameterName=WhiteColor, defaultValue=<rgba 1 1 1 0>}.RGB
   B=@n1.RGB
   Alpha=SmoothStep(Value=Saturate(None=@n2)) {constMax=0.1}
  )
  MaterialFunctionCall(Emissive=@n3) {materialFunction=&MF_FX_EmissiveAdjustments}
  Multiply(A=@n3, B=@n1.RGB)
  n4 = ComponentMask(None=TextureCoordinate()) {r=true}
  n5 = SmoothStep(Value=OneMinus(None=@n4)) {constMin=0.35}
  n6 = LinearInterpolate(
   A=Constant() {r=1}
   B=Multiply(A=SmoothStep(Value=@n4) {constMax=0.25}, B=Multiply(A=@n5, B=@n5))
   Alpha=ScalarParameter() {defaultValue=1, parameterName=UseStartEndFade}
  )
  n7 = ComponentMask(None=TextureCoordinate()) {g=true}
  n8 = SmoothStep(Value=Multiply(A=OneMinus(None=@n7), B=@n7)) {constMax=0.1}
  n9 = MaterialFunctionCall(
   CenterPosition=Constant2Vector() {g=1, r=1}
   Radius=ScalarParameter() {defaultValue=1.25, parameterName=SphereFadeRadius}
   Density=Constant() {r=4}
  ) {materialFunction=&RadialGradientExponential}
  Multiply(A=@n1.A, B=Multiply(A=Multiply(A=@n6, B=@n8), B=OneMinus(None=@n9.RadialGradientExponential)))
  VectorParameter() {parameterName=RedColor, defaultValue=<rgba 1 0 0.0502 1>}
 ]
}
