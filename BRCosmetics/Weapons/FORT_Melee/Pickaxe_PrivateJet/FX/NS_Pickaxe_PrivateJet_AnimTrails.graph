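# Niagara system for the PrivateJet pickaxe's animation trails, drawing four short-lived ribbon layers from the weapon's skeletal mesh.
# Hard and soft green trails share one bone/socket pair, while a narrow white tip trail and StrippedTrail share another.
# FN_AnimTrail modules spawn, initialize and update particles, with User.DebugSpeed added as velocity.
# Color curves fade alpha; all but the hard green layer also set spawn color by time of day.
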NiagaraSystem /BRCosmetics/Weapons/FORT_Melee/Pickaxe_PrivateJet/FX/NS_Pickaxe_PrivateJet_AnimTrails
dependencies {
 usedDataInterfaces [&NiagaraDataInterfaceCurve, &NiagaraDataInterfaceColorCurve]
 usedDynamicInputs []
 usedModules [
  &self:SystemSpawnScript.SystemScriptSource.SystemScriptGraph.NiagaraNodeAssignment_0.SetVariables_#1
  &SystemState
 ]
}
emitters {
 NS_GardGreenTrail {
  emitterUpdateScript [
   &EmitterState()
   &FN_SpawnAnimTrail()
  ]
  particleSpawnScript [
   &FN_InitializeAnimTrailBoneIndecies(Second Bone or Socket Index=3)
   &FN_InitializeAnimTrail(Lifetime=0.1, SkeletalMesh=$System.SkeletalMesh)
   &AddVelocity(Velocity=$User.DebugSpeed)
  ]
  particleUpdateScript [
   &FN_UpdateAnimTrail(SkeletalMesh=$System.SkeletalMesh) {module=FN_UpdateAnimTrail001}
   &Color(Color=&ColorFromCurve(ColorCurve=DataInterface() {propertyValues {
       AlphaCurve {keys [<at 0 0.3>,<at 0.6442 0 RCIM_Cubic>]}
       BlueCurve {keys [<at 0 0.0101>]}
       GreenCurve {keys [<at 0 0.0365>]}
       RedCurve {keys [<at 0 0.0108>]}
      }}))
   &SolveForcesAndVelocity(Clamp Velocity=false, Limit Acceleration=false)
  ]
  rendererClasses [&NiagaraRibbonRendererProperties]
  renderers [{rendererClass=&NiagaraRibbonRendererProperties}]
 }
 NS_SoftGreenTrail {
  emitterUpdateScript [
   &EmitterState()
   &FN_SpawnAnimTrail()
  ]
  particleSpawnScript [
   &FN_InitializeAnimTrailBoneIndecies(Second Bone or Socket Index=3)
   &FN_InitializeAnimTrail(
    Color=&FN_Select_Color_from_Time_Of_Day(Morning=<rgba 0 2.8891 4 1>, Day=<rgba 0 3.8853 5 1>, Evening=<rgba 0 3.5105 4 1>, Night=<rgba 0 3.8978 5 1>)
    Lifetime=0.1
    SkeletalMesh=$System.SkeletalMesh
   )
   &AddVelocity(Velocity=$User.DebugSpeed)
  ]
  particleUpdateScript [
   &FN_UpdateAnimTrail(SkeletalMesh=$System.SkeletalMesh) {module=FN_UpdateAnimTrail001}
   &Color(Color=&ColorFromCurve(ColorCurve=DataInterface() {propertyValues {
       AlphaCurve {keys [<at 0 0.7 RCIM_Cubic>,<at 0.5541 0 RCIM_Cubic>]}
       BlueCurve {keys [<at 0 0.0156>]}
       GreenCurve {keys [<at 0 0.0521>]}
       RedCurve {keys [<at 0 0.0271>]}
      }}))
   &SolveForcesAndVelocity(Clamp Velocity=false, Limit Acceleration=false)
  ]
  rendererClasses [&NiagaraRibbonRendererProperties]
  renderers [{rendererClass=&NiagaraRibbonRendererProperties}]
 }
 NS_TipWhiteTrail {
  emitterUpdateScript [
   &EmitterState()
   &FN_SpawnAnimTrail(AnimTrailWidthMode=::FromSecond)
  ]
  particleSpawnScript [
   &FN_InitializeAnimTrailBoneIndecies(First Bone or Socket Index=1, Second Bone or Socket Index=2)
   &FN_InitializeAnimTrail(
    AnimTrailWidth=0.1
    Color=&FN_Select_Color_from_Time_Of_Day(Morning=<rgba 1 1 1 1>, Day=<rgba 1 1 1 1>, Evening=<rgba 1 1 1 1>, Night=<rgba 1 1 1 1>)
    Lifetime=0.1
    SkeletalMesh=$System.SkeletalMesh
   )
   &AddVelocity(Velocity=$User.DebugSpeed)
  ]
  particleUpdateScript [
   &FN_UpdateAnimTrail(SkeletalMesh=$System.SkeletalMesh) {module=FN_UpdateAnimTrail001}
   &Color(Color=&ColorFromCurve(ColorCurve=DataInterface() {propertyValues {
       AlphaCurve {keys [<at 0 0.6>,<at 0.3748 0 RCIM_Cubic>]}
       BlueCurve {keys [<at 0 1>]}
       GreenCurve {keys [<at 0 1>]}
       RedCurve {keys [<at 0 1>]}
      }}))
   /SetVariables(Particles.Color=&MultiplyLinearColor(
     Linear Color=&MultiplyLinearColor(Linear Color=$Particles.Color, Scale Factor=<1 1 1 1>)
     Scale Factor=&MakeVector4(X=1, Y=1, Z=1, W=1)
    )) {isSetParameters=true}
   &SolveForcesAndVelocity(Clamp Velocity=false, Limit Acceleration=false)
  ]
  rendererClasses [&NiagaraRibbonRendererProperties]
  renderers [{rendererClass=&NiagaraRibbonRendererProperties}]
 }
 StrippedTrail {
  emitterUpdateScript [
   &EmitterState()
   &FN_SpawnAnimTrail()
  ]
  particleSpawnScript [
   &FN_InitializeAnimTrailBoneIndecies(First Bone or Socket Index=1, Second Bone or Socket Index=2)
   &FN_InitializeAnimTrail(
    Color=&FN_Select_Color_from_Time_Of_Day(Morning=<rgba 1 1 1 1>, Day=<rgba 1 1 1 1>, Evening=<rgba 1 1 1 1>, Night=<rgba 1 1 1 1>)
    Lifetime=0.1
    SkeletalMesh=$System.SkeletalMesh
   )
   &AddVelocity(Velocity=$User.DebugSpeed)
  ]
  particleUpdateScript [
   &SolveForcesAndVelocity(Clamp Velocity=false, Limit Acceleration=false)
   &FN_UpdateAnimTrail(SkeletalMesh=$System.SkeletalMesh) {module=FN_UpdateAnimTrail001}
   &ScaleColor(
    Scale Alpha=&FloatFromCurve(
     FloatCurve=DataInterface() {propertyValues {Curve {keys [
         {arriveTangent=-2.003, interpMode=RCIM_Cubic, leaveTangent=-2.003, tangentMode=RCTM_User, time=0.5, value=0.95}
         {interpMode=RCIM_Cubic, tangentMode=RCTM_User, time=1, value=0}
        ]}}}
     Scale Curve=0.1
    )
    ScaleA=true
    ScaleRGB=true
   )
   &Color(Color=&ColorFromCurve(ColorCurve=DataInterface() {propertyValues {
       AlphaCurve {keys [<at 0.0142 0.225>,<at 0.4518 0>]}
       BlueCurve {keys [<at 0 0.1 RCIM_Cubic>]}
       GreenCurve {keys [<at 0 0 RCIM_Cubic>]}
       RedCurve {keys [<at 0 1 RCIM_Cubic>]}
      }}))
  ]
  rendererClasses [&NiagaraRibbonRendererProperties]
  renderers [{rendererClass=&NiagaraRibbonRendererProperties}]
 }
}
summary {
 systemName=NS_Pickaxe_PrivateJet_AnimTrails
 emitters [
  {emitterName=StrippedTrail, rendererClasses [&NiagaraRibbonRendererProperties]}
  {emitterName=NS_SoftGreenTrail, rendererClasses [&NiagaraRibbonRendererProperties]}
  {emitterName=NS_GardGreenTrail, rendererClasses [&NiagaraRibbonRendererProperties]}
  {emitterName=NS_TipWhiteTrail, rendererClasses [&NiagaraRibbonRendererProperties]}
 ]
 userVariables [
  User.Color A {type=LinearColor, description="", defaultValue {struct=LinearColor, value=<rgba 1 1 1 1>}}
  User.DebugSpeed {type=Vector3f, description="", defaultValue {struct=Vector3f, value=<0 0 0>}}
  User.RibbonIDOffset {type=NiagaraInt32}
  User.SkeletalMesh {
   type=NiagaraDataInterfaceSkeletalMesh
   defaultValue {
    struct=NiagaraExt_VariableValue_DataInterface
    value {dataInterface=&self:NiagaraDataInterfaceSkeletalMesh_1}
   }
  }
  User.ToonRibbonPanSpeed {type=NiagaraFloat, defaultValue=-2, description=""}
  User.VariantPicker {type=NiagaraFloat, defaultValue=0, description=""}
 ]
}
