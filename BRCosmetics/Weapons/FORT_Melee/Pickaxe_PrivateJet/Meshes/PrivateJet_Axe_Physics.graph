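# Physics asset for the PrivateJet pickaxe, giving collision bodies and joints to its handle and three umbrella-fold bone chains, apparently for secondary motion.
# Each chain links a tiny MSTR capsule to the two-handed handle capsule with free swing and twist.
# A larger, offset dyn capsule hangs off each MSTR body through a joint with tighter 30-degree swing and twist limits.
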PhysicsAsset /BRCosmetics/Weapons/FORT_Melee/Pickaxe_PrivateJet/Meshes/PrivateJet_Axe_Physics
bodies {
 MSTR_umbrellaFoldGrp_00 [{shapeType=Capsule, length=0.505, radius=0.505, shapeName=MSTR_umbrellaFoldGrp_00_capsule, rotation=<rot 0 -90 0>}]
 MSTR_umbrellaFoldGrp_01 [{shapeType=Capsule, length=0.505, radius=0.505, shapeName=MSTR_umbrellaFoldGrp_01_capsule, rotation=<rot 0 -90 0>}]
 MSTR_umbrellaFoldGrp_02 [{shapeType=Capsule, length=0.505, radius=0.505, shapeName=MSTR_umbrellaFoldGrp_02_capsule, rotation=<rot 0 -90 0>}]
 Melee_TwoHanded_Handle [{shapeType=Capsule, length=1.2559, radius=5.7065, center=<0.2108 1.8188 32.2065>, rotation=<rot 0.4326 -90 0.2617>}]
 dyn_umbrellaFoldGrp_00 [{
   shapeType=Capsule
   length=25.5
   radius=4.5
   shapeName=MSTR_umbrellaFoldGrp_00_capsule
   center=<13.5461 0 0>
   rotation=<rot 0 90.0003 99.9993>
  }]
 dyn_umbrellaFoldGrp_01 [{
   shapeType=Capsule
   length=25.5
   radius=4.5
   shapeName=MSTR_umbrellaFoldGrp_01_capsule
   center=<15.6681 0 0>
   rotation=<rot 0 -89.9986 79.999>
  }]
 dyn_umbrellaFoldGrp_02 [{
   shapeType=Capsule
   length=26
   radius=5
   shapeName=MSTR_umbrellaFoldGrp_02_capsule
   center=<13.8298 0 0>
   rotation=<rot 0.0001 -90.0011 79.9986>
  }]
}
constraints [
 {
  bone1Name=MSTR_umbrellaFoldGrp_00
  bone2Name=Melee_TwoHanded_Handle
  swing1LimitDegrees=45
  swing1Motion=Free
  swing2LimitDegrees=45
  swing2Motion=Free
  twistLimitDegrees=45
  twistMotion=Free
 }
 {
  bone1Name=MSTR_umbrellaFoldGrp_01
  bone2Name=Melee_TwoHanded_Handle
  swing1LimitDegrees=45
  swing1Motion=Free
  swing2LimitDegrees=45
  swing2Motion=Free
  twistLimitDegrees=45
  twistMotion=Free
 }
 {
  bone1Name=MSTR_umbrellaFoldGrp_02
  bone2Name=Melee_TwoHanded_Handle
  swing1LimitDegrees=45
  swing1Motion=Free
  swing2LimitDegrees=45
  swing2Motion=Free
  twistLimitDegrees=45
  twistMotion=Free
 }
 {bone1Name=dyn_umbrellaFoldGrp_00, bone2Name=MSTR_umbrellaFoldGrp_00, swing1LimitDegrees=30, swing2LimitDegrees=30, twistLimitDegrees=30}
 {bone1Name=dyn_umbrellaFoldGrp_01, bone2Name=MSTR_umbrellaFoldGrp_01, swing1LimitDegrees=30, swing2LimitDegrees=30, twistLimitDegrees=30}
 {bone1Name=dyn_umbrellaFoldGrp_02, bone2Name=MSTR_umbrellaFoldGrp_02, swing1LimitDegrees=30, swing2LimitDegrees=30, twistLimitDegrees=30}
]
