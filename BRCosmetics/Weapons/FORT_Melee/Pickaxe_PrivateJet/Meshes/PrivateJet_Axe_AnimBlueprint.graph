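# Animation Blueprint for the PrivateJet pickaxe that adds physics-driven secondary motion to the weapon's bones.
# A cached input pose feeds both a RigidBody node and a chain of three spring-driven Anim Dynamics nodes on the umbrella fold bones.
# bRigidBodyEnabled selects RigidBody when true and the Anim Dynamics chain otherwise, whose alpha drops while crouching or crouch-sprinting.
# The EventGraph holds no connected logic.
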AnimBlueprint /BRCosmetics/Weapons/FORT_Melee/Pickaxe_PrivateJet/Meshes/PrivateJet_Axe_AnimBlueprint
target_skeleton=/Game/Weapons/FORT_Melee/Meshes/SK_Melee_Skeleton
graphs [
 AnimGraph {nodes [
   n1 = RigidBody(ComponentPose=Local To Component(LocalPose=Use cached pose 'RBan_on'())) {
    bOverrideWorldGravity=true
    overridePhysicsAsset=&/PrivateJet_Axe_Physics
    simulationSpace=BaseBoneSpace
    alphaBoolBlend {blendOption=Linear}
    componentLinearAccScale=<0.5 0.25 0.25>
    componentLinearVelScale=<0.25 0.125 0.125>
    overrideWorldGravity {z=-980}
   }
   n2 = SelectFloat(A=0.075, B=0.45, bPickA=OR Boolean(A=Get bIsCrouching(), B=Get bIsCrouchSprinting()))
   n3 = Anim Dynamics - Bone: dyn_umbrellaFoldGrp_00(ComponentPose=Local To Component(LocalPose=Use cached pose 'RBan_on'()), Alpha=@n2, ExternalForce="0,0,0") {
    angularDampingOverride=0.9999
    angularSpringConstant=500
    bAngularSpring=true
    bOverrideAngularDamping=true
    bUseGravityOverride=true
    simulationSpace=RootRelative
    alphaBoolBlend {blendOption=Linear}
    boundBone {boneName=dyn_umbrellaFoldGrp_00}
    componentAppliedLinearAccClamp=<3000 3000 3000>
    componentLinearAccScale=<0.25 0.5 0.25>
    componentLinearVelScale=<0.125 0.25 0.125>
    gravityOverride {z=-980}
    physicsBodyDefinitions [{
      boundBone {boneName=dyn_umbrellaFoldGrp_00}
      boxExtents=<2 10 10>
      constraintSetup {angularLimitsMax=<0 30 30>, angularLimitsMin=<0 -35 -35>, angularTarget=<0 0 0>}
      localJointOffset=<-3 0 -3>
     }]
    relativeSpaceBone {boneName=b_AR_Root}
   }
   n4 = Anim Dynamics - Bone: dyn_umbrellaFoldGrp_01(ComponentPose=@n3, Alpha=@n2, ExternalForce="0,0,0") {
    angularDampingOverride=0.999
    angularSpringConstant=500
    bAngularSpring=true
    bOverrideAngularDamping=true
    bUseGravityOverride=true
    simulationSpace=RootRelative
    alphaBoolBlend {blendOption=Linear}
    boundBone {boneName=dyn_umbrellaFoldGrp_01}
    componentAppliedLinearAccClamp=<3000 3000 3000>
    componentLinearAccScale=<0.25 0.5 0.25>
    componentLinearVelScale=<0.125 0.25 0.125>
    gravityOverride {z=-980}
    physicsBodyDefinitions [{
      boundBone {boneName=dyn_umbrellaFoldGrp_01}
      boxExtents=<2 10 10>
      constraintSetup {angularLimitsMax=<20 30 50>, angularLimitsMin=<-20 -35 -50>, angularTarget=<0 0 0>}
      localJointOffset=<-3 -1.5 3>
     }]
    relativeSpaceBone {boneName=b_AR_Root}
   }
   n5 = Anim Dynamics - Bone: dyn_umbrellaFoldGrp_02(ComponentPose=@n4, Alpha=@n2, ExternalForce="0,0,0") {
    angularDampingOverride=0.9999
    angularSpringConstant=500
    bAngularSpring=true
    bOverrideAngularDamping=true
    bUseGravityOverride=true
    simulationSpace=RootRelative
    alphaBoolBlend {blendOption=Linear}
    boundBone {boneName=dyn_umbrellaFoldGrp_02}
    componentAppliedLinearAccClamp=<3000 3000 3000>
    componentLinearAccScale=<0.25 0.5 0.25>
    componentLinearVelScale=<0.125 0.25 0.125>
    gravityOverride {z=-980}
    physicsBodyDefinitions [{
      boundBone {boneName=dyn_umbrellaFoldGrp_02}
      boxExtents=<2 10 10>
      constraintSetup {angularLimitsMax=<50 30 70>, angularLimitsMin=<-50 -35 -50>, angularTarget=<0 0 0>}
      localJointOffset=<0.5 3 0>
     }]
    relativeSpaceBone {boneName=b_AR_Root}
   }
   Output Pose(Result=Blend Poses by bool(BlendPose_0=Component To Local(ComponentPose=@n1), BlendPose_1=Component To Local(ComponentPose=@n5), bActiveValue=Get bRigidBodyEnabled()) {blendType=Linear}) {layerGroup=None}
   Save cached pose 'RBan_on'(Pose=Input Pose()) {cacheName=RBan_on}
  ]}
 EventGraph {nodes [
   TryGetPawnOwner()
   Event BlueprintUpdateAnimation()
  ]}
]
variables []
